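# Conversation database for Doctor Slone's Season 25 Story Part 3 Quest 1 dialogue, briefing the player on Stingray's stolen schematic beneath Lonely Labs.
# One entry tag leads into four linear Speech nodes advanced by Continue choices.
# The first line is an auto-select main-questline choice, gated on the NPC's Slone tag and an active quest_s25_stories_p03_q01_03 with obj0 incomplete.
# The final choice carries a quest-specific gameplay tag.
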ConversationDatabase /BattlePassS25/Conversations/Stories/Part3/Conversation_S25_Stories_P03_Q01
entry_points [{tag=Conversation.Entry.Season25.Slone.P03.Q01, destinations [#1]}]
nodes [
 ConversationEntryPointNode_0 {
  id=#1
  class=ConversationEntryPointNode
  outputs [#2]
  properties {entryTag {tagName=Conversation.Entry.Season25.Slone.P03.Q01}}
 }
 Speech_C_1 {
  id=#2
  class=Speech_C
  outputs [#3]
  properties {softTaskIcon=&T_UI_NPC_MainQuest, generalConfig {defaultMessage=You impressed me. And I don't say that lightly.}}
  sub_nodes [
   ConversationChoiceNode_2 {
    class=ConversationChoiceNode
    properties {
     defaultChoiceDisplayText=Talk to Slone
     choiceTags {gameplayTags [{tagName=Conversation.Choice.AutoSelectIfValid}, {tagName=Conversation.Choice.Category.MainQuestline}]}
    }
   }
   FortConversationRequirement_ControllerRequirement_1 {
    class=FortConversationRequirement_ControllerRequirement
    properties {
     nodeName=Require: NPC Has Conversation Tag
     participantId {tagName=Conversation.Participant.NPC}
     requirement {class=FortControllerRequirement_NPCHasConversationEntryTag, properties {queryTag {tagName=Conversation.Entry.DoctorSlone}}}
    }
   }
   FortConversationRequirement_ControllerRequirement_3 {
    class=FortConversationRequirement_ControllerRequirement
    properties {
     nodeName=Require: Is Quest Active
     requirement {
      class=FortControllerRequirement_IsQuestActive
      properties {questDefinition=/BattlePassS25/Items/QuestItems/Stories/Part03/quest_s25_stories_p03_q01_03}
     }
    }
   }
   FortConversationRequirement_ControllerRequirement_5 {
    class=FortConversationRequirement_ControllerRequirement
    properties {
     nodeName=Require: Has Completed Quest Objective
     requirement {
      class=FortControllerRequirement_HasCompletedQuestObjective
      properties {
       bInverseCondition=true
       objectiveBackendName=quest_s25_stories_p03_q01_03_obj0
       questDefinition=&quest_s25_stories_p03_q01_03
      }
     }
    }
   }
  ]
 }
 Speech_C_28 {
  id=#3
  class=Speech_C
  outputs [#4]
  properties {generalConfig {defaultMessage="A schematic that's crucial to my plans has ended up in the hands of a two-bit criminal named \"Stingray\"."}}
  sub_nodes [ConversationChoiceNode_8 {class=ConversationChoiceNode, properties {defaultChoiceDisplayText=Continue}}]
 }
 Speech_C_4 {
  id=#4
  class=Speech_C
  outputs [#5]
  properties {generalConfig {defaultMessage="It belonged to Rift Warden Stellan, the creator of that failed rift gate near The Citadel."}}
  sub_nodes [ConversationChoiceNode_0 {class=ConversationChoiceNode, properties {defaultChoiceDisplayText=Continue}}]
 }
 Speech_C_2 {
  id=#5
  class=Speech_C
  outputs []
  properties {generalConfig {defaultMessage="The schematic's well-guarded, in a secret facility under Lonely Labs. Be careful, it's a maze down there."}}
  sub_nodes [ConversationChoiceNode_7 {
    class=ConversationChoiceNode
    properties {defaultChoiceDisplayText=Continue, choiceTags {gameplayTags [{tagName=Athena.Quests.S25.Conversationa.Slone.Choice.P03.Q01}]}}
   }]
 }
]
