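# Conversation graph for Doctor Slone's dialogue in the Season 25 story quest, Part 3 Quest 03 step 01, sending the player after refined Kinetic Ore to supercharge the Apparatus.
# A tagged entry leads to an auto-selected main-questline choice, gated on the NPC having Slone's entry tag, the quest being active and its first objective incomplete.
# Four speech nodes chain linearly through Continue choices, with the final choice carrying a quest-specific gameplay tag.
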ConversationDatabase /BattlePassS25/Conversations/Stories/Part3/Conversation_S25_Stories_P03_Q03_01
entry_points [{tag=Conversation.Entry.Season25.Slone.P03.Q03.01, destinations [#1]}]
nodes [
 ConversationEntryPointNode_0 {
  id=#1
  class=ConversationEntryPointNode
  outputs [#2]
  properties {entryTag {tagName=Conversation.Entry.Season25.Slone.P03.Q03.01}}
 }
 Speech_C_1 {
  id=#2
  class=Speech_C
  outputs [#3]
  properties {
   softTaskIcon=&T_UI_NPC_MainQuest
   generalConfig {defaultMessage="Good work. You're the tenth hired gun I've sent down there, and the first to come back."}
  }
  sub_nodes [
   ConversationChoiceNode_2 {
    class=ConversationChoiceNode
    properties {
     defaultChoiceDisplayText=Talk to Slone
     choiceTags {gameplayTags [{tagName=Conversation.Choice.AutoSelectIfValid}, {tagName=Conversation.Choice.Category.MainQuestline}]}
    }
   }
   FortConversationRequirement_ControllerRequirement_1 {
    class=FortConversationRequirement_ControllerRequirement
    properties {
     nodeName=Require: NPC Has Conversation Tag
     participantId {tagName=Conversation.Participant.NPC}
     requirement {class=FortControllerRequirement_NPCHasConversationEntryTag, properties {queryTag {tagName=Conversation.Entry.DoctorSlone}}}
    }
   }
   FortConversationRequirement_ControllerRequirement_3 {
    class=FortConversationRequirement_ControllerRequirement
    properties {
     nodeName=Require: Is Quest Active
     requirement {
      class=FortControllerRequirement_IsQuestActive
      properties {questDefinition=/BattlePassS25/Items/QuestItems/Stories/Part03/quest_s25_stories_p03_q03_01}
     }
    }
   }
   FortConversationRequirement_ControllerRequirement_5 {
    class=FortConversationRequirement_ControllerRequirement
    properties {
     nodeName=Require: Has Completed Quest Objective
     requirement {
      class=FortControllerRequirement_HasCompletedQuestObjective
      properties {
       bInverseCondition=true
       objectiveBackendName=quest_s25_stories_p03_q03_01_obj1
       questDefinition=&quest_s25_stories_p03_q03_01
      }
     }
    }
   }
  ]
 }
 Speech_C_28 {
  id=#3
  class=Speech_C
  outputs [#4]
  properties {generalConfig {defaultMessage=The Rift Warden never figured out how to refine Kinetic Ore himself.}}
  sub_nodes [ConversationChoiceNode_1 {class=ConversationChoiceNode, properties {defaultChoiceDisplayText=Continue}}]
 }
 Speech_C_2 {
  id=#4
  class=Speech_C
  outputs [#5]
  properties {generalConfig {defaultMessage="But, using his research, I can ping the island for traces of its energy signature."}}
  sub_nodes [ConversationChoiceNode_0 {class=ConversationChoiceNode, properties {defaultChoiceDisplayText=Continue}}]
 }
 Speech_C_0 {
  id=#5
  class=Speech_C
  outputs []
  properties {generalConfig {defaultMessage="If you can track down enough refined Kinetic Ore, we can supercharge the Apparatus."}}
  sub_nodes [ConversationChoiceNode_4 {
    class=ConversationChoiceNode
    properties {defaultChoiceDisplayText=Continue, choiceTags {gameplayTags [{tagName=Athena.Quests.S25.Conversationa.Slone.Choice.P03.Q03.01}]}}
   }]
 }
]
