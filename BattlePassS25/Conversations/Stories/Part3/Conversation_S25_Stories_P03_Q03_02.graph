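# Season 25 Battle Pass story conversation in which Slone briefs the player on charging The Apparatus with refined Kinetic Ore.
# An entry tag leads into a linear chain of four Speech nodes advanced by Continue choices.
# The main-questline opener auto-selects when valid, requiring the Doctor Slone NPC tag and quest p03_q03_05 active with obj0 incomplete.
# The final Continue choice carries a quest-specific gameplay tag.
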ConversationDatabase /BattlePassS25/Conversations/Stories/Part3/Conversation_S25_Stories_P03_Q03_02
entry_points [{tag=Conversation.Entry.Season25.Slone.P03.Q03.02, destinations [#1]}]
nodes [
 ConversationEntryPointNode_0 {
  id=#1
  class=ConversationEntryPointNode
  outputs [#2]
  properties {entryTag {tagName=Conversation.Entry.Season25.Slone.P03.Q03.02}}
 }
 Speech_C_1 {
  id=#2
  class=Speech_C
  outputs [#3]
  properties {
   softTaskIcon=&T_UI_NPC_MainQuest
   generalConfig {defaultMessage=This should be enough refined Kinetic Ore to supercharge The Apparatus.}
  }
  sub_nodes [
   ConversationChoiceNode_2 {
    class=ConversationChoiceNode
    properties {
     defaultChoiceDisplayText=Talk to Slone
     choiceTags {gameplayTags [{tagName=Conversation.Choice.AutoSelectIfValid}, {tagName=Conversation.Choice.Category.MainQuestline}]}
    }
   }
   FortConversationRequirement_ControllerRequirement_1 {
    class=FortConversationRequirement_ControllerRequirement
    properties {
     nodeName=Require: NPC Has Conversation Tag
     participantId {tagName=Conversation.Participant.NPC}
     requirement {class=FortControllerRequirement_NPCHasConversationEntryTag, properties {queryTag {tagName=Conversation.Entry.DoctorSlone}}}
    }
   }
   FortConversationRequirement_ControllerRequirement_3 {
    class=FortConversationRequirement_ControllerRequirement
    properties {
     nodeName=Require: Is Quest Active
     requirement {
      class=FortControllerRequirement_IsQuestActive
      properties {questDefinition=/BattlePassS25/Items/QuestItems/Stories/Part03/quest_s25_stories_p03_q03_05}
     }
    }
   }
   FortConversationRequirement_ControllerRequirement_5 {
    class=FortConversationRequirement_ControllerRequirement
    properties {
     nodeName=Require: Has Completed Quest Objective
     requirement {
      class=FortControllerRequirement_HasCompletedQuestObjective
      properties {
       bInverseCondition=true
       objectiveBackendName=quest_s25_stories_p03_q03_05_obj0
       questDefinition=&quest_s25_stories_p03_q03_05
      }
     }
    }
   }
  ]
 }
 Speech_C_28 {
  id=#3
  class=Speech_C
  outputs [#4]
  properties {generalConfig {defaultMessage="The Apparatus is not a telescope: it's a stellar projector, holding ancient knowledge."}}
  sub_nodes [ConversationChoiceNode_1 {class=ConversationChoiceNode, properties {defaultChoiceDisplayText=Continue}}]
 }
 Speech_C_2 {
  id=#4
  class=Speech_C
  outputs [#5]
  properties {generalConfig {defaultMessage="I've been... arguing with the explorers.  If they see me near the device, they'll overreact."}}
  sub_nodes [ConversationChoiceNode_0 {class=ConversationChoiceNode, properties {defaultChoiceDisplayText=Continue}}]
 }
 Speech_C_0 {
  id=#5
  class=Speech_C
  outputs []
  properties {generalConfig {defaultMessage="Charge up The Apparatus with our refined Kinetic Ore, then return here to power it up."}}
  sub_nodes [ConversationChoiceNode_4 {
    class=ConversationChoiceNode
    properties {defaultChoiceDisplayText=Continue, choiceTags {gameplayTags [{tagName=Athena.Quests.S25.Conversationa.Slone.Choice.P03.Q03.02}]}}
   }]
 }
]
